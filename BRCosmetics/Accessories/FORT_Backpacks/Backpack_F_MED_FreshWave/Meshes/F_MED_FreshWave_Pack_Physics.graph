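# Physics asset for the FreshWave backpack cosmetic, setting up physics for its dangling straps, belt, triOrn ornament and zipper pull.
# One large capsule covers the pack, tiny capsules sit on the base bones as anchors, and fitted capsules cover each dyn bone.
# Each dyn part has tight angular limits against its anchor, with the zipper as a two-link chain.
# Anchor-to-pack constraints leave swing and twist free.
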PhysicsAsset /BRCosmetics/Accessories/FORT_Backpacks/Backpack_F_MED_FreshWave/Meshes/F_MED_FreshWave_Pack_Physics
bodies {
 base_dyn_baseStrap_l [{shapeType=Capsule, length=1, radius=0.1, shapeName=base_dyn_belt_capsule, rotation=<rot 0 -90 0>}]
 base_dyn_baseStrap_r [{shapeType=Capsule, length=1, radius=0.1, shapeName=base_dyn_belt_capsule, rotation=<rot 0 -90 0>}]
 base_dyn_belt [{shapeType=Capsule, length=1, radius=0.1, shapeName=base_dyn_belt_capsule, rotation=<rot 0 -90 0>}]
 base_dyn_triOrn [{shapeType=Capsule, length=1, radius=0.1, shapeName=base_dyn_triOrn_capsule, rotation=<rot 0 -90 0>}]
 base_dyn_zipper_1 [{shapeType=Capsule, length=1, radius=0.1, shapeName=base_dyn_zipper_1_capsule, rotation=<rot 0 -90 0>}]
 dyn_baseStrap_l [{
   shapeType=Capsule
   length=3.5558
   radius=0.5293
   shapeName=dyn_baseStrap_r_capsule
   center=<1.801 0 -0.6281>
   rotation=<rot 50.1224 -0.0001 -0.0002>
  }]
 dyn_baseStrap_r [{
   shapeType=Capsule
   length=3.5558
   radius=0.5293
   shapeName=dyn_baseStrap_r_capsule
   center=<1.7988 0 -0.6262>
   rotation=<rot 50.1224 -0.0001 -0.0002>
  }]
 dyn_belt [{shapeType=Capsule, length=9.0941, radius=3.279, shapeName=dyn_belt_capsule, center=<8.9791 -0.8311 3.074>, rotation=<rot -11.2905 0 -90>}]
 dyn_triOrn [{
   shapeType=Capsule
   length=4.1296
   radius=0.8445
   shapeName=dyn_triOrn_capsule
   center=<2.5455 -0.0309 0.073>
   rotation=<rot 0 -90.6264 -89.7795>
  }]
 dyn_zipper_1 [{
   shapeType=Capsule
   length=3.3894
   radius=0.505
   shapeName=dyn_zipper_1_capsule
   center=<1.9585 0.0001 -0.3552>
   rotation=<rot 0 -89.3613 -90.7785>
  }]
 dyn_zipper_2 [{shapeType=Capsule, length=5.3105, radius=0.5, shapeName=dyn_zipper_2_capsule, center=<2.7188 0.019 0.3774>, rotation=<rot 0 89.4786 -89.446>}]
 pack [{shapeType=Capsule, length=8.6583, radius=12.2388, shapeName=pack_capsule, center=<0 -0.7891 -2.7091>, rotation=<rot 0 -90 -90>}]
}
constraints [
 {bone1Name=dyn_zipper_2, bone2Name=dyn_zipper_1, swing1LimitDegrees=10, swing2LimitDegrees=50, twistLimitDegrees=0}
 {bone1Name=dyn_zipper_1, bone2Name=base_dyn_zipper_1, swing1LimitDegrees=20, swing2LimitDegrees=30, twistLimitDegrees=0}
 {bone1Name=dyn_triOrn, bone2Name=base_dyn_triOrn, swing1LimitDegrees=40, swing2LimitDegrees=0, twistLimitDegrees=0}
 {bone1Name=dyn_belt, bone2Name=base_dyn_belt, swing1LimitDegrees=5, swing2LimitDegrees=45, twistLimitDegrees=3}
 {
  bone1Name=base_dyn_belt
  bone2Name=pack
  swing1LimitDegrees=45
  swing1Motion=Free
  swing2LimitDegrees=45
  swing2Motion=Free
  twistLimitDegrees=45
  twistMotion=Free
 }
 {
  bone1Name=base_dyn_zipper_1
  bone2Name=pack
  swing1LimitDegrees=45
  swing1Motion=Free
  swing2LimitDegrees=45
  swing2Motion=Free
  twistLimitDegrees=45
  twistMotion=Free
 }
 {
  bone1Name=base_dyn_triOrn
  bone2Name=pack
  swing1LimitDegrees=45
  swing1Motion=Free
  swing2LimitDegrees=45
  swing2Motion=Free
  twistLimitDegrees=45
  twistMotion=Free
 }
 {bone1Name=dyn_baseStrap_l, bone2Name=base_dyn_baseStrap_l, swing1LimitDegrees=4, swing2LimitDegrees=12, twistLimitDegrees=4}
 {bone1Name=dyn_baseStrap_r, bone2Name=base_dyn_baseStrap_r, swing1LimitDegrees=4, swing2LimitDegrees=12, twistLimitDegrees=4}
 {
  bone1Name=base_dyn_baseStrap_r
  bone2Name=pack
  swing1LimitDegrees=45
  swing1Motion=Free
  swing2LimitDegrees=45
  swing2Motion=Free
  twistLimitDegrees=45
  twistMotion=Free
 }
 {
  bone1Name=base_dyn_baseStrap_l
  bone2Name=pack
  swing1LimitDegrees=45
  swing1Motion=Free
  swing2LimitDegrees=45
  swing2Motion=Free
  twistLimitDegrees=45
  twistMotion=Free
 }
]
